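# Translucent UI material for the PlacidMelon reticle brackets, with a percent fill and cooldown, no-ammo and color-change states.
# Mirrored MSDF prong textures take spread and position offsets, while a circle-outline SDF intersected with a cosine band forms the curved bars.
# A three-color gradient pulses on an eased time sine and blends toward cyan-white or gray by state.
# PercentFill drives a vertical step mask, and named reroutes carry masks into color and opacity.
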Material /PlacidMelon/Gameplay/PlacidMelon/UI/Materials/M_UI_PlacidMelon_Brackets
graph {
 properties {bAutomaticallySetUsageInEditor=false, blendMode=BLEND_Translucent, materialDomain=MD_UI, usedShadingModels=Default Lit}
 nodes [
  n1 = NamedRerouteUsage()
  n2 = LinearInterpolate(
   A=Constant4Vector() {constant=<rgba 1 1 1 1>}.RGBA
   B=Constant4Vector() {constant=<rgba 0 0 0 1>}.RGBA
   Alpha=ScalarParameter() {parameterName=Cooldown}
  ) {constAlpha=0}
  n3 = NamedRerouteUsage()
  n4 = LinearInterpolate(
   A=Constant4Vector() {constant=<rgba 0.812 0 0 1>}.RGBA
   B=Constant4Vector() {constant=<rgba 0 0 0 1>}.RGBA
   Alpha=ScalarParameter() {parameterName=Cooldown}
  ) {constAlpha=0}
  Add(A=Multiply(A=Multiply(A=@n1, B=@n2), B=Constant() {r=1}), B=Multiply(A=Multiply(A=@n3, B=@n4), B=Constant() {r=1}))
  n5 = TextureCoordinate()
  n6 = MaterialFunctionCall(Speed=Constant() {r=0.1}) {materialFunction=&DebugTimeSine}
  n7 = Power(
   Base=Add(A=ComponentMask(None=@n5) {g=true}) {constB=0.3}
   Exp=LinearInterpolate(Alpha=MaterialFunctionCall("Value (0-1)"=@n6.Result) {materialFunction=&MF_UI_EaseCurves}.EaseInOutSine) {constA=2.5, constB=7}
  ) {constExponent=5.5}
  n8 = TextureCoordinate()
  n9 = Power(Base=Add(A=ComponentMask(None=@n8) {g=true}) {constB=0.38}) {constExponent=10}
  n10 = LinearInterpolate(
   A=MaterialFunctionCall(
    A=VectorParameter() {parameterName=Color01, defaultValue=<rgba 1 0 0.0049 1>}.RGB
    B=VectorParameter() {parameterName=Color03, defaultValue=<rgba 0 0 0 1>}.RGB
    C=VectorParameter() {parameterName=Color02, defaultValue=<rgba 1 0 0 1>}.RGB
    Alpha=Multiply(A=@n7)
   ) {materialFunction=&Lerp_3Color}.Result
   B=LinearInterpolate(A=Constant3Vector() {constant=<rgba 0.0437 0.956 1 1>}, B=Constant3Vector() {constant=<rgba 1 1 1 0>}, Alpha=Multiply(A=@n9))
   Alpha=ScalarParameter() {parameterName=ColorChange}
  )
  n11 = Constant3Vector() {constant=<rgba 0.5 0.5 0.5 1>}
  n12 = ScalarParameter() {parameterName=NoAmmo}
  LinearInterpolate(
   A=LinearInterpolate(A=ScalarParameter() {parameterName=ShadowColor}, B=LinearInterpolate(A=@n10, B=@n11, Alpha=@n12), Alpha=NamedRerouteUsage())
   B=VectorParameter() {parameterName=DetailColor, defaultValue=<rgba 1 1 1 0>}.RGB
   Alpha=NamedRerouteUsage()
  )
  n13 = TextureCoordinate()
  n14 = Add(A=ScalarParameter() {defaultValue=-0.25, parameterName=XOffset}, B=Constant() {r=0.32})
  n15 = Constant() {r=-0.18}
  MaterialFunctionCall(
   UVs=Add(
    A=MaterialFunctionCall!(UV=@n13, Swizzle UV) {materialFunction=/FortUILibrary/MaterialFunctions/Basics/MF_UI_Symmetry_2x2_New}.Symmetry
    B=AppendVector(A=@n14, B=@n15)
   )
   Texture Scale=Constant() {r=0.18}
  ) {materialFunction=&ScaleUVsByCenter}
  n16 = NamedRerouteUsage()
  n17 = MaterialFunctionCall!(UV=TextureCoordinate(), Mirror Y, Swizzle UV) {materialFunction=/FortUILibrary/MaterialFunctions/Basics/MF_UI_Symmetry_2x1_New}
  n18 = AppendVector(A=ScalarParameter() {defaultValue=-0.25, parameterName=XOffset}, B=Constant())
  n19 = MaterialFunctionCall(
   UVs=MaterialFunctionCall(UVs=Add(A=@n17.Symmetry, B=@n18), Texture Scale=Constant2Vector() {g=0.3, r=0.6}) {materialFunction=&ScaleUVsByCenter}.UVs
   Size=Constant() {r=1}
   OutlineThickness=ScalarParameter() {defaultValue=0.04, parameterName=OutlineThickness}
  ) {materialFunction=&MF_UI_Circle}
  n20 = Cosine(None=MaterialFunctionCall() {materialFunction=&GetUserInterfaceUV}.Normalized UV)
  n21 = Add(A=ComponentMask(None=@n20) {g=true}) {constB=0.6}
  n22 = MaterialFunctionCall(
   DistanceField=MaterialFunctionCall!(DistanceFieldA=@n21, DistanceFieldB=@n19.SDFOutline, BlendAmount=Constant() {r=0.15}) {materialFunction=/FortUILibrary/MaterialFunctions/Shapes/Modifiers/MF_UI_SDF_Intersection}.Smooth
   Threshold=ScalarParameter() {defaultValue=0.03, parameterName=ShadowThickness}
  ) {materialFunction=&MF_UI_AntiAliasing}
  n23 = ScalarParameter() {defaultValue=0.3, parameterName=ShadowOpacity}
  Max(
   A=Subtract(A=LinearInterpolate(A=@n16, B=@n22.Inner, Alpha=@n23), B=NamedRerouteUsage())
   B=Multiply(A=NamedRerouteUsage(), B=ScalarParameter() {defaultValue=0.5, parameterName=DetailOpacity}) {constB=0.5}
  )
  n24 = Add(
   A=MaterialFunctionCall(UV=TextureCoordinate()) {materialFunction=&MF_UI_Symmetry_2x1_New}.Symmetry
   B=AppendVector(A=Constant() {r=-0.09}, B=ScalarParameter() {defaultValue=0.007, parameterName=Y-Outer})
  ) {constB=0}
  n25 = Constant2Vector() {g=0.3, r=0.7}
  n26 = Multiply(A=ScalarParameter() {defaultValue=0.1, desc=Variable for Accuracy Spread Data, parameterName=SpreadOffsetOuter}) {constB=-1, desc=Invert}
  n27 = Constant() {desc=Variable to Position Reticle Prongs}
  n28 = Add(
   A=MaterialFunctionCall(
    UVs=MaterialFunctionCall(UVs=@n24, Texture Scale=@n25) {materialFunction=&ScaleUVsByCenter}.UVs
    "Rotation Angle (0-1)"=Constant() {r=0.75}
   ) {materialFunction=&CustomRotator}.Rotated Values
   B=AppendVector(A=Constant(), B=Add(A=@n26, B=@n27)) {desc=Inverts SpreadOffset to Keep Center Bars In Place}
  ) {constB=-0.1, desc=Add Offset}
  n29 = ScalarParameter() {defaultValue=0.7, parameterName=ScaleOuter}
  n30 = MaterialFunctionCall!(MSDF RGB=TextureSample(UVs=MaterialFunctionCall(UVs=@n28, Texture Scale=@n29) {materialFunction=&ScaleUVsByCenter}.UVs) {texture=/PlacidMelon/Gameplay/PlacidMelon/UI/Textures/T_UI_MaskShape_Outer_1}.RGB) {materialFunction=/FortUILibrary/MaterialFunctions/Shapes/Modifiers/MF_UI_MSDFtoSDF}
  n31 = LinearInterpolate(
   A=MaterialFunctionCall(DistanceField=@n30.SDF, Threshold=Constant() {r=0.03}) {materialFunction=&MF_UI_AntiAliasing}.Outer
   B=MaterialFunctionCall(DistanceField=@n30.SDF, Threshold=Constant() {r=0.12}) {materialFunction=&MF_UI_AntiAliasing}.Outer
  )
  n32 = ScalarParameter() {parameterName=Cooldown}
  n33 = Add(
   A=MaterialFunctionCall(UV=TextureCoordinate()) {materialFunction=&MF_UI_Symmetry_2x1_New}.Symmetry
   B=AppendVector(A=Constant(), B=Constant() {r=-0.005})
  ) {constB=0}
  n34 = Constant2Vector() {g=0.45, r=1}
  n35 = ScalarParameter() {desc=Variable for Accuracy Spread Data, group=Reticle Operators, parameterName=SpreadOffsetInner, sliderMax=1, sliderMin=-1}
  n36 = Multiply(A=@n35) {constB=-1, desc=Invert}
  n37 = ScalarParameter() {desc=Variable to Position Reticle Prongs, group=Reticle Operators, parameterName=PositionalOffset, sliderMax=1, sliderMin=-1}
  n38 = Add(
   A=MaterialFunctionCall(
    UVs=MaterialFunctionCall(UVs=@n33, Texture Scale=@n34) {materialFunction=&ScaleUVsByCenter}.UVs
    "Rotation Angle (0-1)"=Constant() {r=0.75}
   ) {materialFunction=&CustomRotator}.Rotated Values
   B=AppendVector(A=Constant(), B=Add(A=@n36, B=@n37)) {desc=Inverts SpreadOffset to Keep Center Bars In Place}
  ) {constB=-0.1, desc=Add Offset}
  n39 = ScalarParameter() {defaultValue=1, group=Prong Operators, parameterName=ScaleInner, sliderMax=2}
  n40 = MaterialFunctionCall(MSDF RGB=TextureSample(UVs=MaterialFunctionCall(UVs=@n38, Texture Scale=@n39) {materialFunction=&ScaleUVsByCenter}.UVs) {texture=/PlacidMelon/Gameplay/PlacidMelon/UI/Textures/T_UI_MaskShape}.RGB) {materialFunction=&MF_UI_MSDFtoSDF}
  n41 = MaterialFunctionCall(DistanceField=@n40.SDF, Threshold=Constant() {r=0.1}) {materialFunction=&MF_UI_AntiAliasing}
  n42 = NamedRerouteDeclaration(None=MaterialFunctionCall(DistanceField=@n40.SDF, Threshold=Constant() {r=0.14}) {materialFunction=&MF_UI_AntiAliasing}.Outer) {name=OuterThin, nodeColor=<rgba 0.0781 0 1 1>}
  n43 = ScalarParameter() {parameterName=Cooldown}
  Max(
   A=Multiply(
    A=NamedRerouteDeclaration(None=@n31) {name=Inner, nodeColor=<rgba 0 0.8125 1 1>}
    B=LinearInterpolate(Alpha=@n32) {constA=0.7, constAlpha=0, constB=0.2}
   ) {constB=0.7}
   B=Multiply(A=LinearInterpolate(A=@n41.Outer, B=@n42) {constAlpha=0.7}, B=LinearInterpolate(Alpha=@n43) {constA=0.65, constAlpha=0, constB=0.2}) {constB=0.65}
  )
  n44 = MaterialFunctionCall(DistanceFieldA=@n21, DistanceFieldB=@n19.SDFOutline, BlendAmount=Constant() {r=0.08}) {materialFunction=&MF_UI_SDF_Intersection}
  Multiply(A=@n44.Smooth)
  NamedRerouteDeclaration(None=MaterialFunctionCall(DistanceField=@n44.Smooth) {materialFunction=&MF_UI_AntiAliasing}.Inner) {name=ColoredBar, nodeColor=<rgba 0 0 0 1>}
  n45 = SmoothStep(
   Min=ScalarParameter() {defaultValue=0.7, parameterName=SmoothStepMin}
   Max=ScalarParameter() {defaultValue=0.77, parameterName=SmoothStepMax}
  )
  NamedRerouteDeclaration(None=@n45) {name=Detail, nodeColor=<rgba 0 0 0 1>}
  n46 = Ceil(None=Add(A=ComponentMask(None=@n20) {r=true}) {constB=0.5})
  NamedRerouteDeclaration(None=Saturate(None=OneMinus(None=@n46))) {name=Mask, nodeColor=<rgba 0 0 0 1>}
  n47 = ScalarParameter() {parameterName=SmoothStepMinShadow}
  n48 = ScalarParameter() {defaultValue=0.1, parameterName=SmoothStepMaxShadow}
  NamedRerouteDeclaration(None=LinearInterpolate(A=@n45, B=SmoothStep(Min=@n47, Max=@n48), Alpha=ScalarParameter() {defaultValue=0.3, parameterName=ShadowOpacity})) {name=DetailOpacity, nodeColor=<rgba 0 0 0 1>}
  n49 = NamedRerouteUsage()
  n50 = Subtract(A=@n49) {constB=0.7}
  n51 = ComponentMask(None=MaterialFunctionCall() {materialFunction=&GetUserInterfaceUV}.Normalized UV) {g=true}
  n52 = ScalarParameter() {defaultValue=1, parameterName=PercentFill}
  n53 = Constant()
  n54 = Constant() {r=1}
  n55 = Constant() {r=0.3565}
  n56 = Constant() {r=0.6455}
  n57 = Multiply(
   A=@n49
   B=Step(
    Y=OneMinus(None=@n51)
    X=MaterialFunctionCall(Input=@n52, Input Low=@n53, Input High=@n54, Target Low=@n55, Target High=@n56) {materialFunction=&RemapValueRange}.Result
   )
  )
  NamedRerouteDeclaration(None=Saturate(None=Add(A=@n50, B=@n57))) {name=BracketFinalOpacity, nodeColor=<rgba 0 0 0 1>}
  NamedRerouteDeclaration(None=@n57) {name=BracketColorAlpha, nodeColor=<rgba 0 0 0 1>}
 ]
}
