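# PlacidMelon weapon reticle widget, built on the recharge-stack reticle base, showing brackets and a center dot driven by dynamic materials.
# It plays an equip animation on construct and an attack animation whenever the shield-placing gameplay tag is added to the player controller.
# Local ammo changes drive bracket fill and play or reverse the no-ammo dimming animation.
# A timer polls item cooldown for pip recharge, though the GetPipCollection override returns no collection.
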WidgetBlueprint /PlacidMelon/Gameplay/PlacidMelon/UI/WBP_PlacidMelon_Reticle
animations [
 Anim_Attack {
  binding_count=2
  length_seconds=0.45
  object=&self:Anim_Attack
  sequence {bindings [
    Image_Reticle {
     id=#1
     possessed_class=FortLazyImage
     tracks [{
       type=MovieSceneWidgetMaterialTrack
       sections [{
         type=MovieSceneParameterSection
         end_frame=9
         start_frame=0
         channels [
          XOffset {type=MovieSceneScriptingFloatChannel, keys [[0,-0.25],[3.9563,-0.2516],[5.659,-0.1504],[9,-0.25]]}
          ColorChange {type=MovieSceneScriptingFloatChannel, keys [[2,0],[6,1],[9,0]]}
          OutlineThickness {type=MovieSceneScriptingFloatChannel, keys [[2,0.06],[6,0.08],[9,0.06]]}
         ]
        }]
      }]
    }
    Image_CenterDot {
     id=#2
     possessed_class=Image
     tracks [{
       type=MovieSceneWidgetMaterialTrack
       sections [{
         type=MovieSceneParameterSection
         end_frame=9
         start_frame=0
         channels [
          CircleCrossColorR {type=MovieSceneScriptingFloatChannel, keys [[2,0.658],[6,0.5],[9,0.658]]}
          CircleCrossColorG {type=MovieSceneScriptingFloatChannel, keys [[2,0.584],[6,0.956],[9,0.584]]}
          CircleCrossColorB {type=MovieSceneScriptingFloatChannel, const [1,3]}
          CircleCrossColorA {type=MovieSceneScriptingFloatChannel, const [1,1]}
         ]
        }]
      }]
    }
   ]}
 }
 Anim_Equip {
  binding_count=2
  length_seconds=0.25
  object=&self:Anim_Equip
  sequence {bindings [
    Image_Reticle {
     id=#3
     possessed_class=FortLazyImage
     tracks [
      {
       type=MovieSceneWidgetMaterialTrack
       sections [{
         type=MovieSceneParameterSection
         end_frame=5
         start_frame=0
         channels [
          XOffset {type=MovieSceneScriptingFloatChannel, keys [[0,-0.06],[5,-0.25]]}
          ShadowColor {type=MovieSceneScriptingFloatChannel, keys [[0,1],[5,0]]}
          ColorChange {type=MovieSceneScriptingFloatChannel, keys [[0,1],[2,1],[5,0]]}
         ]
        }]
      }
      {
       type=MovieSceneFloatTrack
       display_name=Render Opacity
       sections [{type=MovieSceneFloatSection, end_frame=5, start_frame=0, channels [None {type=MovieSceneScriptingFloatChannel, keys [[0,0],[5,1]]}]}]
      }
     ]
    }
    Image_CenterDot {
     id=#4
     possessed_class=Image
     tracks [{
       type=MovieSceneFloatTrack
       display_name=Render Opacity
       sections [{type=MovieSceneFloatSection, end_frame=5, start_frame=0, channels [None {type=MovieSceneScriptingFloatChannel, keys [[0,0],[5,1]]}]}]
      }]
    }
   ]}
 }
 Anim_NoAmmo {
  binding_count=2
  length_seconds=0.25
  object=&self:Anim_NoAmmo
  sequence {bindings [
    Image_CenterDot {
     id=#5
     possessed_class=Image
     tracks [{
       type=MovieSceneWidgetMaterialTrack
       sections [{
         type=MovieSceneParameterSection
         end_frame=5
         start_frame=0
         channels [
          ReticleOpacity {type=MovieSceneScriptingFloatChannel, keys [[0,0.85],[5,0.5]]}
          CircleCrossColorR {type=MovieSceneScriptingFloatChannel, keys [[0,0.658],[5,0.1]]}
          CircleCrossColorG {type=MovieSceneScriptingFloatChannel, keys [[0,0.584],[5,0.1]]}
          CircleCrossColorB {type=MovieSceneScriptingFloatChannel, keys [[0,1],[5,0.1]]}
          CircleCrossColorA {type=MovieSceneScriptingFloatChannel, const [1,2]}
         ]
        }]
      }]
    }
    Image_Reticle {
     id=#6
     possessed_class=FortLazyImage
     tracks [
      {
       type=MovieSceneFloatTrack
       display_name=Render Opacity
       sections [{type=MovieSceneFloatSection, end_frame=5, start_frame=0, channels [None {type=MovieSceneScriptingFloatChannel, keys [[0,1],[5,0.5]]}]}]
      }
      {
       type=MovieSceneWidgetMaterialTrack
       sections [{type=MovieSceneParameterSection, end_frame=5, start_frame=0, channels [NoAmmo {type=MovieSceneScriptingFloatChannel, keys [[0,0],[5,1]]}]}]
      }
     ]
    }
   ]}
 }
]
graphs [
 BindToWeaponEvents {nodes [
   n1 = BindToWeaponEvents!() -> (then, DoBind:bool, EquippedWeapon:Fort Weapon Object Reference) {is=title K2Node_FunctionEntry}
   > Parent: BindToWeaponEvents!(execute, DoBind:bool=@n1, EquippedWeapon:Fort Weapon Object Reference=@n1.EquippedWeapon) -> (then) {is=title K2Node_CallParentFunction}
   Reroute Node(InputPin:Fort Weapon Object Reference=@n1.EquippedWeapon) -> (OutputPin:Fort Weapon Object Reference)
  ]}
 EventGraph {nodes [
   OnEquip() {is=title K2Node_CustomEvent}
   > ListenForAttack()
   > PlayAnimationForward(InAnimation=Get Anim_Equip() -> (Anim_Equip:Widget Animation Object Reference))
   > n1 = GetDynamicMaterial(self=Get Image_Reticle() -> (Image_Reticle:Fort Lazy Image Object Reference))
   > n2 = Set MID Reticle(MID Reticle:Material Instance Dynamic Object Reference=@n1) -> (Output_Get:Material Instance Dynamic Object Reference)
   > SetScalarParameterValue(self=@n2, ParameterName=PercentFill, Value=100)
   > n3 = GetDynamicMaterial(self=Get Image_CenterDot() -> (Image_CenterDot:Image Object Reference))
   > Set MID_CenterDot(MID_CenterDot:Material Instance Dynamic Object Reference=@n3) -> (Output_Get:Material Instance Dynamic Object Reference)
   > OnChangedBracketColors!(execute, self, HasAmmo:bool=true) -> (then)
   ListenForAttack() {is=title K2Node_CustomEvent}
   > n4 = WaitGameplayTagAddToActor!(
    execute:Exec
    TargetActor:Actor Object Reference=Get PlayerController() -> (PlayerController:Player Controller Object Reference)
    Tag:Gameplay Tag Structure=Get ShieldPlacingGameplayTag() -> (ShieldPlacingGameplayTag:Gameplay Tag Structure)
    OnlyTriggerOnce:bool=false
   ) -> (then:Exec, AsyncTaskProxy:Ability Async Wait Gameplay Tag Added Object Reference, Added:Exec) {is=title K2Node_AsyncAction}
   HandleAttack(execute=@n4.Added)
   n5 = Event OnLocalAmmoChanged!() -> (OutputDelegate, then, LocalCount:int=0, LocalRemaining:int=0) {is=title K2Node_Event}
   > Parent: OnLocalAmmoChanged!(execute, LocalCount:int=@n5.LocalCount, LocalRemaining:int=@n5.LocalRemaining) -> (then) {is=title K2Node_CallParentFunction}
   > UpdateCooldownQueryTimer(self:Fort Weapon Reticle Extension Widget Base Object Reference)
   > HandleChargeCountAnimations!(execute, self, Local Count:int=@n5.LocalCount) -> (then)
   > Set PreviousChargeCount(PreviousChargeCount:int=@n5.LocalCount) -> (Output_Get:int=0)
   > n6 = GetDynamicMaterial(self=Get Image_Reticle())
   > SetScalarParameterValue(self=@n6, ParameterName=PercentFill, Value=float / float(B=100))
   Event Construct()
   > Parent: Construct()
   > Set PlayerController(PlayerController:Player Controller Object Reference=GetOwningPlayer()) -> (Output_Get:Player Controller Object Reference)
   > OnEquip() {is=title K2Node_CallFunction}
   > UpdateCooldownQueryTimer()
   Event UpdateRemainingCooldown() {is=title K2Node_Event}
   > n7 = GetRemainingCooldownTimer!(execute, self) -> (then, RemainingTime:double, TotalDuration:double)
   > UpdateRemainingRechargeTime!(
    execute
    self:Fort Weapon Reticle Recharge Stack Base Pip Collection Base Object Reference=GetPipCollection!(self:WBP Reticle Recharge Stack Base Object Reference) -> (PipCollection:WBP Reticle Recharge Stack Base Pip Collection Object Reference)
    TimeRemaining:float=@n7
    TotalDuration:float=@n7.TotalDuration
    PipPercentMatParamName:Name=Progress
    ShouldSetPipColor:bool=false
    NotReadyColor:Linear Color Structure
    ReadyColor:Linear Color Structure
    PipColorMatParamName:Name=None
   ) -> (then)
   Event Destruct()
   > BindToWeaponEvents!(execute, self:WBP Reticle Recharge Stack Base Object Reference, DoBind:bool=false, EquippedWeapon:Fort Weapon Object Reference) -> (then) {is=title K2Node_CallFunction}
   > HandleCooldownTimer!(execute, self:Fort Weapon Reticle Extension Widget Base Object Reference, bActivateTimer:bool=false, TimerInterval:float=0) -> (then)
   Event UpdateCooldownQueryTimer() {is=title K2Node_Event}
   > Is Valid()
   > n8 = Branch()
   > HandleCooldownTimer(bActivateTimer=false, TimerInterval=0)
   HandleCooldownTimer(execute=@n8.else, bActivateTimer=true, TimerInterval=0.1)
   Reroute Node(InputPin:Wildcard) -> (OutputPin:Wildcard)
   Reroute Node()
   float / float(B=100)
  ]}
 GetPipCollection {nodes [
   GetPipCollection() {is=title K2Node_FunctionEntry}
   > Return Node!(execute, PipCollection:WBP Reticle Recharge Stack Base Pip Collection Object Reference)
   Parent: GetPipCollection!() -> (PipCollection:WBP Reticle Recharge Stack Base Pip Collection Object Reference) {is=title K2Node_CallParentFunction}
  ]}
 GetRemainingCooldownTimer {nodes [
   GetRemainingCooldownTimer() {is=title K2Node_FunctionEntry}
   > n1 = Cast To FortPlayerController(Object=GetOwningPlayer())
   Return Node!(execute=[@n2.Is Not Valid, @n1.CastFailed], RemainingTime:double=0, TotalDuration:double=0)
   n2 = Is Valid()
   n3 = GetCoolDownTimeRemaining!(self:Fort Item Object Reference, FortPC:Fort Player Controller Object Reference=@n1) -> (CooldownDuration:float, ReturnValue:float)
   > Return Node(RemainingTime=@n3.ReturnValue, TotalDuration=@n3)
   Reroute Node()
  ]}
 HandleAttack {nodes [
   HandleAttack() {is=title K2Node_FunctionEntry}
   > PlayAnimationForward(InAnimation=Get Anim_Attack() -> (Anim_Attack:Widget Animation Object Reference), PlaybackSpeed=0.5)
  ]}
 HandleChargeCountAnimations {nodes [
   n1 = HandleChargeCountAnimations!() -> (then, Local Count:int) {is=title K2Node_FunctionEntry}
   > n2 = Branch(Condition="integer > integer"(A=@n1))
   > Branch(Condition="Equal (Integer)"(A=Get PreviousChargeCount() -> (PreviousChargeCount:int=0)))
   > PlayAnimationReverse(InAnimation=Get Anim_NoAmmo() -> (Anim_NoAmmo:Widget Animation Object Reference))
   PlayAnimationForward(execute=@n2.else, InAnimation=Get Anim_NoAmmo())
  ]}
 OnChangedBracketColors {nodes [
   n1 = OnChangedBracketColors!() -> (then, HasAmmo:bool=~) {is=title K2Node_FunctionEntry}
   > SetScalarParameterValue(
    self=Get MID Reticle() -> (MID Reticle:Material Instance Dynamic Object Reference)
    ParameterName=NoAmmo
    Value=SelectFloat(A=0, B=0, bPickA=@n1)
   )
   > SetVectorParameterValue(
    self=Get MID_CenterDot() -> (MID_CenterDot:Material Instance Dynamic Object Reference)
    ParameterName=CircleCrossColor
    Value=SelectColor(A=(R=0.6584,G=0.5841,B=1,A=1), B=(R=0.0999,G=0.0999,B=0.0999,A=1), bPickA=Get HasAmmo() -> (HasAmmo:bool))
   )
   > n2 = GetDynamicMaterial(self=Get Image_CenterDot())
   > SetScalarParameterValue(self=@n2, ParameterName=ReticleOpacity, Value=SelectFloat(A=0.85, B=0.5, bPickA=Get HasAmmo()))
   > SetRenderOpacity(self=Get Image_Reticle(), InOpacity=SelectFloat(A=1, B=0.5, bPickA=Get HasAmmo()))
  ]}
]
variables [
 MID Reticle {type=MaterialInstanceDynamic}
 MID_CenterDot {type=MaterialInstanceDynamic}
 PlayerController {type=PlayerController, default=null}
 PreviousChargeCount {type=integer}
 ShieldPlacingGameplayTag {type=GameplayTag, default {tagName=None}}
]
widgets {
 info {
  inheritedWidgetCount=1
  parentClass=/Game/Athena/HUD/reticles/ReticleExtensionWidgets/ChargeStack/WBP_Reticle_RechargeStackBase.WBP_Reticle_RechargeStackBase_C
  rootWidgetClass=&Overlay
  widgetCount=4
 }
 widgets [
  ReticleOverlay {widgetClassPath=&Overlay, widget=ReticleOverlay}
  Image_Reticle {
   widgetClassPath=&FortLazyImage
   bIsVariable=true
   parent=ReticleOverlay
   widget=Image_Reticle
   properties {
    brush {
     bIsDynamicallyLoaded=false
     drawAs=Image
     imageType=FullColor
     mirroring=NoMirror
     resourceName=None
     resourceObject=/PlacidMelon/Gameplay/PlacidMelon/UI/Materials/M_UI_PlacidMelon_Brackets
     tiling=NoTile
     imageSize=<135 135>
     margin=<ltrb 0.5 0.5 0.5 0.5>
     outlineSettings {
      bUseBrushTransparency=false
      roundingType=HalfHeightRadius
      width=0
      color {colorUseRule=UseColor_Specified, specifiedColor=<rgba 0 0 0 0>}
      cornerRadii=<0 0 0 0>
     }
     tintColor {colorUseRule=UseColor_Specified, specifiedColor=<rgba 1 1 1 1>}
     uVRegion {bIsValid=false, max=<0 0>, min=<0 0>}
    }
    renderTransform {angle=0, scale=<1.2 1.2>, shear=<0 0>, translation=<0 0>}
   }
   slot {class=OverlaySlot, properties {horizontalAlignment=HAlign_Center}}
  }
  Image_CenterDot {
   widgetClassPath=&Image
   bIsVariable=true
   parent=ReticleOverlay
   widget=Image_CenterDot
   properties {
    brush {
     bIsDynamicallyLoaded=false
     drawAs=Image
     imageType=FullColor
     mirroring=NoMirror
     resourceName=None
     resourceObject=/PlacidMelon/Gameplay/PlacidMelon/UI/Materials/M_Reticle_Center_PlacidMelon
     tiling=NoTile
     imageSize=<10 10>
     margin=<ltrb 0 0 0 0>
     outlineSettings {
      bUseBrushTransparency=false
      roundingType=HalfHeightRadius
      width=0
      color {colorUseRule=UseColor_Specified, specifiedColor=<rgba 0 0 0 0>}
      cornerRadii=<0 0 0 0>
     }
     tintColor {colorUseRule=UseColor_Specified, specifiedColor=<rgba 1 1 1 1>}
     uVRegion {bIsValid=false, max=<0 0>, min=<0 0>}
    }
    renderTransform {angle=0, scale=<2 2>, shear=<0 0>, translation=<0 0>}
   }
   slot {class=OverlaySlot, properties {horizontalAlignment=HAlign_Center, verticalAlignment=VAlign_Center}}
  }
  TutorialHighlight {widgetClassPath=&TutorialHighlight}
 ]
}
